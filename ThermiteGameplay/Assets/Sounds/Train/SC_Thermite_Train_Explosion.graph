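# Sound cue for the Thermite train explosion, layering a close and a distant blast, routed to the explosions sound class.
# The distant layer mixes a custom distant wave with a random gas pump explosion, adds pitch variation and a short fixed delay, and uses distant attenuation.
# The close wave plays undelayed at slightly higher mix volume through its own average-range attenuation.
# A concurrency override limits the cue to one instance.
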SoundCue /ThermiteGameplay/Assets/Sounds/Train/SC_Thermite_Train_Explosion
root=Mixer_0
nodes [
 {
  id=Attenuation_0
  children [Delay_1]
  properties {attenuationSettings=&/ATT_Thermite_Ecplosion_Train_Distant}
 }
 {
  id=Attenuation_5
  children [Modulator_2]
  properties {attenuationSettings=&/ATT_Thermite_Ecplosion_Train_Average}
 }
 {id=Delay_1, children [Modulator_3], properties {delayMax=0.1, delayMin=0.1}}
 {id=Mixer_0, children [Attenuation_0, Attenuation_5], properties {inputVolume [1,1.2]}}
 {id=Mixer_1, children [Random_1,Modulator_0], properties {inputVolume [1,1]}}
 {id=Modulator_0, children [WavePlayer_0]}
 {id=Modulator_2, children [WavePlayer_2], properties {pitchMax=1, pitchMin=1}}
 {id=Modulator_3, children [Mixer_1], properties {pitchMax=1.1, pitchMin=0.9}}
 {id=Random_1, children [WavePlayer_1,WavePlayer_3], properties {weights [1,1]}}
 {id=WavePlayer_0, properties {soundWaveAssetPtr=&/Thermite_Train_Explosion_1_distant}}
 {id=WavePlayer_1, properties {soundWaveAssetPtr=/Game/Sounds/Fort_Destructibles/GasPump/Destructible_Gas_Pump_EXPLODE_Distant_01}}
 {id=WavePlayer_2, properties {soundWaveAssetPtr=&/Thermite_Train_Explosion_1_Close}}
 {id=WavePlayer_3, properties {soundWaveAssetPtr=/Game/Sounds/Fort_Destructibles/GasPump/Destructible_Gas_Pump_EXPLODE_Distant_02}}
]
properties {
 attenuationSettings=&ATT_Thermite_Ecplosion_Train_Distant
 bOverrideConcurrency=true
 maxDistance=23600
 soundClassObject=/Game/Sounds/FortSoundClassesAndModes/Fort_Events_Explosions
 volumeMultiplier=1
 concurrencyOverrides {maxCount=1}
}
