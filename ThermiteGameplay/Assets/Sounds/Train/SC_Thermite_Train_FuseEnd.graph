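# Sound cue for the end of a Thermite train fuse, mixing a main wave with a louder close-range layer.
# Each layer overrides the shared close explosion-tail attenuation with HRTF spatialization and its own falloff, short for the close layer and much longer for the main one.
# Concurrency allows one instance at a time, stopping the oldest when a new one plays.
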SoundCue /ThermiteGameplay/Assets/Sounds/Train/SC_Thermite_Train_FuseEnd
root=Mixer_0
nodes [
 {
  id=Attenuation_0
  children [WavePlayer_0]
  properties {
   attenuationSettings=/Game/Sounds/Attenuation/Fort_Explosion_Close_Tails_Att
   bOverrideAttenuation=true
   attenuationOverrides {falloffDistance=1000, spatializationAlgorithm=SPATIALIZATION_HRTF, stereoSpread=0, attenuationShapeExtents {x=0}}
  }
 }
 {
  id=Attenuation_1
  children [WavePlayer_1]
  properties {
   attenuationSettings=&Fort_Explosion_Close_Tails_Att
   bOverrideAttenuation=true
   attenuationOverrides {falloffDistance=5000, spatializationAlgorithm=SPATIALIZATION_HRTF, stereoSpread=0, attenuationShapeExtents {x=1000}}
  }
 }
 {id=Mixer_0, children [Attenuation_1, Attenuation_0], properties {inputVolume [1.3,3]}}
 {id=WavePlayer_0, properties {soundWaveAssetPtr=&/Thermite_Train_FuseEnd_Close}}
 {id=WavePlayer_1, properties {soundWaveAssetPtr=&/Thermite_Train_FuseEnd}}
]
properties {bOverrideConcurrency=true, maxDistance=6000, volumeMultiplier=2, concurrencyOverrides {maxCount=1, resolutionRule=StopOldest}}
