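# Sound cue for the ignition sound of the thermite train in the Thermite gameplay content.
# A single wave player plays the default thermite ignite sound through an attenuation node that overrides its explosion-tail preset with HRTF spatialization, no stereo spread and a custom falloff distance.
# The cue boosts volume and overrides concurrency to allow one instance at a time, stopping the oldest.
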SoundCue /ThermiteGameplay/Assets/Sounds/Train/SC_Thermite_Train_Ignite
root=Attenuation_1
nodes [
 {
  id=Attenuation_1
  children [WavePlayer_0]
  properties {
   attenuationSettings=/Game/Sounds/Attenuation/Fort_Explosion_Close_Tails_Att
   bOverrideAttenuation=true
   attenuationOverrides {falloffDistance=6000, spatializationAlgorithm=SPATIALIZATION_HRTF, stereoSpread=0, attenuationShapeExtents {x=0}}
  }
 }
 {id=WavePlayer_0, properties {soundWaveAssetPtr=/ThermiteGameplay/Assets/Sounds/Thermite_Default_Ignite_A1}}
]
properties {bOverrideConcurrency=true, maxDistance=6000, volumeMultiplier=4, concurrencyOverrides {maxCount=1, resolutionRule=StopOldest}}
