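# Sound cue for interacting with the Thermite train, playing its activation sound.
# A single wave player feeds a modulator at default settings into the output.
# It uses mono item-interact attenuation and the loot-collect sound class, with a slightly raised volume.
# A concurrency override caps simultaneous instances and prevents new ones once the limit is reached.
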SoundCue /ThermiteGameplay/Assets/Sounds/Train/SC_Thermite_Train_Interact
root=Modulator_0
nodes [
 {id=Modulator_0, children [WavePlayer_1]}
 {id=WavePlayer_1, properties {soundWaveAssetPtr=&/Thermite_Train_Activation}}
]
properties {
 attenuationSettings=/Game/Sounds/Attenuation/Fort_Item_Interact_Mono_Att
 bOverrideConcurrency=true
 maxDistance=1500
 soundClassObject=&Fort_Loot_Collect
 volumeMultiplier=1.2
 concurrencyOverrides {maxCount=2, resolutionRule=PreventNew, volumeScaleAttackTime=0.1}
}
