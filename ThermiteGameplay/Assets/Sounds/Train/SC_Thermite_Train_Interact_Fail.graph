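# Sound cue for a failed interaction with the Thermite train, spatialized with weapon-pickup attenuation under the gadget items sound class.
# A Mixer layers two long search-start waves at equal volume, one through its own Modulator.
# A root Modulator varies the combined mix, so the second layer gets two stages of default pitch and volume randomization.
# A concurrency override uses PreventNew, so new plays are rejected once the limit is reached.
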SoundCue /ThermiteGameplay/Assets/Sounds/Train/SC_Thermite_Train_Interact_Fail
root=Modulator_0
nodes [
 {id=Mixer_0, children [WavePlayer_0,Modulator_1], properties {inputVolume [1,1]}}
 {id=Modulator_0, children [Mixer_0]}
 {id=Modulator_1, children [WavePlayer_2]}
 {id=WavePlayer_0, properties {soundWaveAssetPtr=&/Thermite_Train_Search_Long_Start_2}}
 {id=WavePlayer_2, properties {soundWaveAssetPtr=&/Thermite_Train_Search_Long_Start}}
]
properties {
 attenuationSettings=&Fort_Weapon_Pickup_Att
 bOverrideConcurrency=true
 maxDistance=3500
 soundClassObject=&Fort_Gadgets_Items
 volumeMultiplier=1
 concurrencyOverrides {resolutionRule=PreventNew}
}
