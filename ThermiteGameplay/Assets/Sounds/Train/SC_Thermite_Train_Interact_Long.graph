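# Sound cue for the long train interaction or activation sound in the Thermite gameplay content.
# A single wave player feeds the Thermite_Train_Activation_Long wave straight to the output.
# It uses Fortnite's item-interact attenuation and player-interact sound class with a boosted volume.
# A concurrency override caps simultaneous instances at a small count.
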SoundCue /ThermiteGameplay/Assets/Sounds/Train/SC_Thermite_Train_Interact_Long
root=WavePlayer_0
nodes [{id=WavePlayer_0, properties {soundWaveAssetPtr=&/Thermite_Train_Activation_Long}}]
properties {
 attenuationSettings=/Game/Sounds/Attenuation/Fort_Item_Interact_Att
 bOverrideConcurrency=true
 maxDistance=1850
 soundClassObject=/Game/Sounds/FortSoundClassesAndModes/Fort_08_Player_Interacts
 volumeMultiplier=1.4
 concurrencyOverrides {maxCount=2, volumeScaleAttackTime=0.1}
}
