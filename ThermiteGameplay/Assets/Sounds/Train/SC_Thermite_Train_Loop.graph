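# Looping sound cue for the Thermite train sound, layering three looping waves under the Fort_Water_Interaction sound class.
# A short-range branch mixes LayerF with the default Thermite loop, while a long-range branch plays LayerA at double pitch.
# Each branch has its own HRTF attenuation override, so the LayerA branch carries much farther.
# Both feed a mixer with an empty third input, then a root modulator slightly randomizes the boosted volume.
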SoundCue /ThermiteGameplay/Assets/Sounds/Train/SC_Thermite_Train_Loop
root=Modulator_0
nodes [
 {
  id=Attenuation_0
  children [Mixer_0]
  properties {
   attenuationSettings=/Game/Sounds/Attenuation/Fort_Player_Swimming_Athena_Att
   bOverrideAttenuation=true
   attenuationOverrides {
    bEnableSubmixSends=true
    dBAttenuationAtMax=-30
    falloffDistance=2000
    spatializationAlgorithm=SPATIALIZATION_HRTF
    stereoSpread=100
    attenuationShapeExtents {x=0}
    submixSendSettings [{
      bEnableHPFCutoff=false
      bEnableLPFCutoff=false
      disableManualSendClamp=false
      hPFCutoff=20
      lPFCutoff=20000
      maxSendDistance=6000
      maxSendLevel=1
      minSendDistance=400
      minSendLevel=0
      sendLevel=0.2
      sendLevelControlMethod=Linear
      soundSubmix=null
     }]
   }
  }
 }
 {
  id=Attenuation_4
  children [Modulator_1]
  properties {
   attenuationSettings=&Fort_Player_Swimming_Athena_Att
   bOverrideAttenuation=true
   attenuationOverrides {
    bEnableSubmixSends=true
    falloffDistance=15000
    spatializationAlgorithm=SPATIALIZATION_HRTF
    stereoSpread=100
    attenuationShapeExtents {x=0}
    submixSendSettings [{
      bEnableHPFCutoff=false
      bEnableLPFCutoff=false
      disableManualSendClamp=false
      hPFCutoff=20
      lPFCutoff=20000
      maxSendDistance=6000
      maxSendLevel=1
      minSendDistance=400
      minSendLevel=0
      sendLevel=0.2
      sendLevelControlMethod=Linear
      soundSubmix=null
     }]
   }
  }
 }
 {id=Mixer_0, children [WavePlayer_10, WavePlayer_0], properties {inputVolume [0.5,1]}}
 {id=Mixer_3, children [Attenuation_0, Attenuation_4, null], properties {inputVolume [2.5,0.6,0.8]}}
 {id=Modulator_0, children [Mixer_3], properties {volumeMax=1.25, volumeMin=1.2}}
 {id=Modulator_1, children [WavePlayer_11], properties {pitchMax=2, pitchMin=2}}
 {id=WavePlayer_0, properties {bLooping=true, soundWaveAssetPtr=/ThermiteGameplay/Assets/Sounds/Thermite_Default_Loop_C}}
 {id=WavePlayer_10, properties {bLooping=true, soundWaveAssetPtr=&/Thermite_Train_LayerF}}
 {id=WavePlayer_11, properties {bLooping=true, soundWaveAssetPtr=&/Thermite_Train_LayerA}}
]
properties {
 maxDistance=15000
 soundClassObject=/Game/Sounds/FortSoundClassesAndModes/Fort_Water_Interaction
 volumeMultiplier=3
 attenuationOverrides {
  bEnableOcclusion=true
  distanceAlgorithm=Logarithmic
  falloffDistance=1800
  nonSpatializedRadiusStart=600
  occlusionLowPassFilterFrequency=18000
  occlusionTraceChannel=ECC_GameTraceChannel11
  occlusionVolumeAttenuation=0.6
  stereoSpread=2000
 }
}
