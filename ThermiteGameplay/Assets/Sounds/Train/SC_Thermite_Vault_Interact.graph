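# SC_Thermite_Vault_Interact is a sound cue for interacting with the vault, stored with the Thermite train sounds.
# A single wave player plays the Thermite_Train_Activation wave through a Modulator into the output.
# It uses pickup-style weapon attenuation and the gadgets/items sound class at reduced volume, with a concurrency override that caps simultaneous instances.
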SoundCue /ThermiteGameplay/Assets/Sounds/Train/SC_Thermite_Vault_Interact
root=Modulator_0
nodes [
 {id=Modulator_0, children [WavePlayer_1]}
 {id=WavePlayer_1, properties {soundWaveAssetPtr=&/Thermite_Train_Activation}}
]
properties {
 attenuationSettings=&Fort_Weapon_Pickup_Att
 bOverrideConcurrency=true
 maxDistance=3500
 soundClassObject=&Fort_Gadgets_Items
 volumeMultiplier=0.5
 concurrencyOverrides {maxCount=2}
}
